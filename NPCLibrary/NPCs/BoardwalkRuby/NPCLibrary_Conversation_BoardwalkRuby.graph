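# Conversation database for the Boardwalk Ruby NPC, defining the greeting shown when a player talks to the character.
# A single entry point leads to a speech node with a default beach line and contextual variants for a first-ever conversation or specific equipped character skins.
# The speech then branches through link nodes into the shared NPC Library quests and basic services conversations, referenced by entry tag.
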ConversationDatabase /NPCLibrary/NPCs/BoardwalkRuby/NPCLibrary_Conversation_BoardwalkRuby
entry_points [{tag=Conversation.Entry.BoardwalkRuby, destinations [#1]}]
nodes [
 ConversationEntryPointNode_7 {
  id=#1
  class=ConversationEntryPointNode
  outputs [#2]
  properties {entryTag {tagName=Conversation.Entry.BoardwalkRuby}}
 }
 Speech_C_4 {
  id=#2
  class=Speech_C
  outputs [#3, #4]
  properties {
   nodeName=""
   generalConfig {
    defaultMessage=I'm sure enjoying these absolutely normal beach activities. You?
    contextualMessages [
     {
      message="Oh hello, yes, hi, how are you, you're good? Good. So am I."
      contextRequirements [{requirement {class=FortControllerRequirement_IsFirstEverConversationWithNPC}}]
      priority {curve {rowName=Tandem.Speech.Priority.FirstEverInteraction}}
     }
     {
      message="If anyone asks, you don't know me."
      requirementMatchPolicy=RequireAny
      contextRequirements [{requirement {
         class=FortControllerRequirement_HasCID
         properties {allowedCIds [
           /BRCosmetics/Athena/Items/Cosmetics/Characters/CID_154_Athena_Commando_M_Gumshoe
           /BRCosmetics/Athena/Items/Cosmetics/Characters/CID_155_Athena_Commando_F_Gumshoe
           /BRCosmetics/Athena/Items/Cosmetics/Characters/CID_159_Athena_Commando_M_GumshoeDark
           /BRCosmetics/Athena/Items/Cosmetics/Characters/CID_694_Athena_Commando_M_CatBurglar
           /BRCosmetics/Athena/Items/Cosmetics/Characters/CID_A_034_Athena_Commando_F_CatBurglar
           /BRCosmetics/Athena/Items/Cosmetics/Characters/CID_A_129_Athena_Commando_M_CatBurglarSummer
           /BRCosmetics/Athena/Items/Cosmetics/Characters/CID_A_257_Athena_Commando_M_CatBurglar_Ghost
           /BRCosmetics/Athena/Items/Cosmetics/Characters/CID_A_320_Athena_Commando_M_CatburglarWinter
           /BRCosmetics/Athena/Items/Cosmetics/Characters/CID_923_Athena_Commando_M_ParcelGold
           &CID_693_Athena_Commando_M_BuffCat
           &CID_701_Athena_Commando_M_BananaAgent
           &CID_692_Athena_Commando_M_HenchmanTough
           /BRCosmetics/Athena/Items/Cosmetics/Characters/CID_A_202_Athena_Commando_F_Division
          ]}
        }}]
      priority {value=0, curve {curveTable=null, rowName=None}}
     }
    ]
   }
  }
 }
 ConversationLinkNode_0 {id=#4, class=ConversationLinkNode, properties {remoteEntryTag {tagName=Conversation.Entry.BasicServices}}}
 ConversationLinkNode_1 {id=#3, class=ConversationLinkNode, properties {remoteEntryTag {tagName=Conversation.Entry.NPCLibrary.Quests}}}
]
